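; Quest gameplay ability for a "damage after hiding" challenge, triggered by a gameplay event when the player exits a hiding prop.
; It casts the event instigator to B_HidingProp and checks its tags against the quest's target hiding prop tags.
; On a match it applies a configured gameplay effect to the owner, apparently a 30-second window, then ends; other cases just end.
; Debug prints and an optional shipping-log helper trace each branch.
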(event Ability|EventActivateAbilityFromEvent
  (params (EventData :type Gameplay%20Event%20Data%20Structure%20%28by%20ref%29))
  (bind (_eventtag _instigator) (Utilities|Struct|BreakGameplayEventData EventData))
  (|CallParentFunction EventData)
  (Development|PrintString "@@@ ActivateAbilityFromEvent [GA_Quest_Damage_afterHiding30]: CALLED!" false (Variables|Debug|GetbDebugAbility))
  (bind _asb_hiding_prop (Utilities|Casting|CastToB_HidingProp _instigator)
    (:then
      (if (GameplayTags|HasAnyMatchingGameplayTags _asb_hiding_prop (Variables|Default|GetQuestTargetPropHidingTags))
        (Development|PrintString "@@@@@@ ActivateAbilityFromEvent [GA_Quest_Damage_afterHiding30]: Exited a target hiding prop!" false (Variables|Debug|GetbDebugAbility))
        (Ability|ApplyGameplayEffectToOwner (Variables|Default|GetGameplay%20Effect%20Class))
        (Default|DebugAbilityLogging true false "GA_Quest_Damage_AfterHiding30: ActivateAbilityFromEvent.ApplyGameplayEffectToOwner: GE_Quest_Damage_AfterHiding30s")
        (Ability|EndAbility)
        (else
          (Development|PrintString "@@@@@@ ActivateAbilityFromEvent [GA_Quest_Damage_afterHiding30]: Exited a hiding prop NOT selected by this mission!" false (Variables|Debug|GetbDebugAbility))
          (Ability|EndAbility))))
    (:CastFailed
      (Ability|EndAbility))))

(fn DebugAbilityLogging
  (params (BuildLogEnabled? :type Boolean :default true) (PIE%20PrintEnabled? :type Boolean) (Message :type String) (GameplayEventData :type Gameplay%20Event%20Data%20Structure%20%28by%20ref%29))
  (bind _instigator (Utilities|Struct|BreakGameplayEventData GameplayEventData))
  (bind _returnvalue (Utilities|GetDisplayName _instigator))
  (bind _returnvalue_2 (Utilities|String|Append Message _returnvalue))
  (Utilities|FlowControl|Sequence
    (:then_0
      (if BuildLogEnabled?
        (Fortnite|FortShippingLog self _returnvalue_2)))
    (:then_1
      (if PIE%20PrintEnabled?
        (Development|PrintString _returnvalue_2)))))
